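# Sound class named Character in the Figment_Martin sound class hierarchy, grouping character and enemy audio.
# It sits under the close enemy attacks and abilities class and parents classes for enemy attack, normal enemy activity and troll sounds.
# It applies effects and sends a small amount of bleed to the low-frequency channel.
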SoundClass /Figment_Martin/Sounds/FortSoundClassesAndModes/Character
parent=&/Fort_07_Close_Enemy_Attacks_Abilities
children [
 &/Fort_Enemy_Attack_PSM
 &/Fort_Enemy_Normal_Activity
 /Game/Sounds/FortSoundClassesAndModes/Fort_Troll
]
properties {properties {bApplyEffects=true, lFEBleed=0.25}}
