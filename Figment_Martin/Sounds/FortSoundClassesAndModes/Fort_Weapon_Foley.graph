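# Fort_Weapon_Foley is a sound class for weapon foley audio in the Figment_Martin content.
# It is a child of the Fort_Weapons sound class, placing it in the weapons branch of the sound class hierarchy.
# It turns on effects processing and sends a partial bleed into the low-frequency effects channel.
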SoundClass /Figment_Martin/Sounds/FortSoundClassesAndModes/Fort_Weapon_Foley
parent=/Game/Sounds/FortSoundClassesAndModes/Fort_Weapons
properties {properties {bApplyEffects=true, lFEBleed=0.25}}
