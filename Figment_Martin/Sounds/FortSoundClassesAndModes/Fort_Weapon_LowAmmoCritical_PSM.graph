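# Sound class for low or critical ammo weapon sounds, as its name indicates, parented to the Fort_WeaponProjectile class.
# While active, it applies the Fort_Weapon_CriticalDuck_MixMod sound mix as a passive modifier once its volume passes a minimum threshold, which by name ducks other audio.
# It also sets a slightly reduced volume, enables effects and adds some LFE bleed.
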SoundClass /Figment_Martin/Sounds/FortSoundClassesAndModes/Fort_Weapon_LowAmmoCritical_PSM
parent=&/Fort_WeaponProjectile
properties {
 passiveSoundMixModifiers [{minVolumeThreshold=0.2, soundMix=/Game/Sounds/FortSoundClassesAndModes/Fort_Weapon_CriticalDuck_MixMod}]
 properties {bApplyEffects=true, lFEBleed=0.25, volume=0.8}
}
