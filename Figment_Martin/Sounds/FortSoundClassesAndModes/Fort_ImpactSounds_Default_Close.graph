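# Sound class for default close-range impact sounds in the Figment_Martin content.
# It is a child of the Fort_ImpactSounds class in the sound class hierarchy.
# It enables effects processing and sends part of the signal to the LFE channel as bleed.
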SoundClass /Figment_Martin/Sounds/FortSoundClassesAndModes/Fort_ImpactSounds_Default_Close
parent=&/Fort_ImpactSounds
properties {properties {bApplyEffects=true, lFEBleed=0.25}}
