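# Sound class for impact sounds that, going by its name, makes default impact sounds quieter while its own sounds play.
# It is a child of Fort_ImpactSounds, with boosted volume, effects enabled and some LFE bleed.
# A passive sound mix modifier automatically activates Fort_DuckDefaultImpactSoundsQuieter_MixMod while a sound in this class plays above a low volume threshold, so no explicit mix push is needed.
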SoundClass /Figment_Martin/Sounds/FortSoundClassesAndModes/Fort_ImpactSounds_DuckDefaultQuieter_PSM
parent=&/Fort_ImpactSounds
properties {
 passiveSoundMixModifiers [{minVolumeThreshold=0.1, soundMix=&/Fort_DuckDefaultImpactSoundsQuieter_MixMod}]
 properties {bApplyEffects=true, lFEBleed=0.2, volume=1.4}
}
