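# Sound class for impact sounds that, as its name indicates, duck the default impact sounds while they play.
# It is a child of Fort_ImpactSounds, with effects enabled, some LFE bleed and boosted volume.
# A passive sound mix modifier automatically applies Fort_DuckDefaultImpactSounds_MixMod while a sound in this class plays above a low volume threshold.
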SoundClass /Figment_Martin/Sounds/FortSoundClassesAndModes/Fort_ImpactSounds_DuckDefault_PSM
parent=&/Fort_ImpactSounds
properties {
 passiveSoundMixModifiers [{minVolumeThreshold=0.05, soundMix=&/Fort_DuckDefaultImpactSounds_MixMod}]
 properties {bApplyEffects=true, lFEBleed=0.2, volume=1.3}
}
