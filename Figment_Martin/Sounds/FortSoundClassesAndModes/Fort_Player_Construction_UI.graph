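# Sound class for player construction UI sounds, judging by its name, in the Figment_Martin sound class and mode set.
# It is parented to the Fort_ImpactSounds class, placing it under that branch of the mix hierarchy.
# It sets its own slightly reduced volume and an LFE bleed amount, and defines no child classes.
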SoundClass /Figment_Martin/Sounds/FortSoundClassesAndModes/Fort_Player_Construction_UI
parent=&/Fort_ImpactSounds
properties {properties {lFEBleed=0.25, volume=0.8}}
