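# Sound class for impact sounds, parented to the player searching, looting and crafting class in the Figment_Martin sound hierarchy.
# It sets a reduced volume and a low-frequency effects bleed.
# Its children split impacts into close, default close, distant and two ducked variants, and also include player construction and construction UI classes, the former referenced from the base game's sound folder.
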SoundClass /Figment_Martin/Sounds/FortSoundClassesAndModes/Fort_ImpactSounds
parent=&/Fort_08_Player_Searching_looting_crafting
children [
 &/Fort_ImpactSounds_Close
 &/Fort_ImpactSounds_Default_Close
 &/Fort_ImpactSounds_Distant
 &/Fort_ImpactSounds_DuckDefault_PSM
 &/Fort_ImpactSounds_DuckDefaultQuieter_PSM
 /Game/Sounds/FortSoundClassesAndModes/Fort_Player_Construction
 &/Fort_Player_Construction_UI
]
properties {properties {lFEBleed=0.25, volume=0.6}}
