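# Sound class for high-tier player loot sounds in Figment_Martin, parented to the Fort player searching, looting and crafting class.
# It passively triggers the Fort_Player_Loot_Duck_LowMed_MixMod sound mix from a zero volume threshold, which by its name ducks low and medium loot sounds.
# It also enables effects processing and sets some LFE bleed.
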SoundClass /Figment_Martin/Sounds/FortSoundClassesAndModes/Fort_Player_Loot_High_PSM
parent=&/Fort_08_Player_Searching_looting_crafting
properties {
 passiveSoundMixModifiers [{minVolumeThreshold=0, soundMix=&/Fort_Player_Loot_Duck_LowMed_MixMod}]
 properties {bApplyEffects=true, lFEBleed=0.25}
}
