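# Sound class grouping player searching, looting and crafting audio, such as doors, searching foley and impact sounds, within the in-game SFX mix.
# It sits under Fort_SFX_InGame and parents six child classes, including searching reward, loot high and loot drop/collect suppress classes.
# Its own properties set a small low-frequency effects bleed.
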SoundClass /Figment_Martin/Sounds/FortSoundClassesAndModes/Fort_08_Player_Searching_looting_crafting
parent=&Fort_SFX_InGame
children [
 &Fort_Doors
 &/Fort_Searching_Foley
 &/Fort_ImpactSounds
 &/Fort_Searching_Reward_PSM
 &/Fort_Player_Loot_High_PSM
 &/Fort_Loot_Drop_Collect_Suppress
]
properties {properties {lFEBleed=0.25}}
